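# ETetsets is a Default Lit test material in the WildEstate environment base materials, with no clear in-game purpose.
# Its graph holds only two red constant color nodes, with no connections shown to each other or to material outputs.
# It is flagged for skeletal mesh and mesh deformer use, with automatic usage setting in the editor turned off.
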Material /WildEstate/Environment/BaseMaterials/Test/ETetsets
graph {
 properties {bAutomaticallySetUsageInEditor=false, bUsedWithMeshDeformer=true, bUsedWithSkeletalMesh=true, usedShadingModels=Default Lit}
 nodes [
  Constant3Vector() {constant=<rgba 1 0.0317 0 1>}
  Constant3Vector() {constant=<rgba 1 0 0 0>}
 ]
}
