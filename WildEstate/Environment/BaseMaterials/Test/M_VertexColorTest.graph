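# M_VertexColorTest is a Default Lit test material for vertex color, with Nanite usage enabled and automatic usage setting in the editor turned off.
# Its graph feeds Vertex Color into the A input of a Multiply node that has no B connection.
# A Constant3Vector node sits unconnected, and no link from either branch to a material output is present.
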Material /WildEstate/Environment/BaseMaterials/Test/M_VertexColorTest
graph {
 properties {bAutomaticallySetUsageInEditor=false, bUsedWithNanite=true, usedShadingModels=Default Lit}
 nodes [
  Constant3Vector()
  Multiply(A=VertexColor())
 ]
}
